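# M_InteriorGlow is a Default Lit material in the MaterialGym folder that wraps the weapon-skin interior glow function, apparently as a test or showcase.
# It uses material attributes, passing a SetMaterialAttributes node with no overrides into MF_InteriorGlow.
# The function's Mask input is a constant one, so no texture or parameter limits where the effect applies.
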Material /Game/Global/Materials/MaterialGym/Materials/InteriorGlow/M_InteriorGlow
graph {
 properties {bUseMaterialAttributes=true, pixelDepthOffsetMode=PDOM_Legacy, usedShadingModels=Default Lit}
 nodes [MaterialFunctionCall!(In=SetMaterialAttributes!(MaterialAttributes), Mask=Constant() {r=1}) {materialFunction=/Game/Weapons/WeaponSkins/MaterialFunction/MF_InteriorGlow}]
}
